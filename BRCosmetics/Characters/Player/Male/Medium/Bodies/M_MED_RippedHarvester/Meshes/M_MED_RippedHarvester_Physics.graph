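# Physics asset for the RippedHarvester medium male player body, set up for a dynamic coat.
# Five coat chains (center, left/right bt, left/right side) each run three capsules down from a small master-joint body, with limited swing and zero twist.
# Free-angular cross constraints tie some neighboring chains, but none links center to left.
# Pelvis, thigh, calf and foot shapes provide body volumes, and a huge offset sphere on the floor bone approximates ground.
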PhysicsAsset /BRCosmetics/Characters/Player/Male/Medium/Bodies/M_MED_RippedHarvester/Meshes/M_MED_RippedHarvester_Physics
bodies {
 calf_l [{shapeType=Capsule, length=45.7953, radius=8.9887, center=<-18.227 -2.3251 0.6745>, rotation=<rot 90 180 1.4636>}]
 calf_r [{shapeType=Capsule, length=61.2764, radius=9.2852, center=<18.227 2.3251 -0.6745>, rotation=<rot 90 -179.7762 -178.3126>}]
 dyn_C_coat_center_1_jnt [{shapeType=Capsule, length=29, radius=9, shapeName=dyn_C_coat_center_1_jnt_capsule, center=<9.6385 0 3.1099>, rotation=<rot -0.0001 0 90>}]
 dyn_C_coat_center_2_jnt [{shapeType=Capsule, length=29, radius=9, shapeName=dyn_C_coat_center_2_jnt_capsule, center=<9.6385 0 2.3937>, rotation=<rot -0.0001 0 90>}]
 dyn_C_coat_center_3_jnt [{shapeType=Capsule, length=29, radius=9, shapeName=dyn_C_coat_center_3_jnt_capsule, center=<9.6385 0 2.3937>, rotation=<rot -0.0001 0 90>}]
 dyn_L_coat_bt_1_jnt [{shapeType=Capsule, length=29, radius=9, shapeName=dyn_L_coat_bt_1_jnt_capsule, center=<9.4921 -1.6737 6.7162>, rotation=<rot 0 -10 90>}]
 dyn_L_coat_bt_2_jnt [{shapeType=Capsule, length=29, radius=9, shapeName=dyn_L_coat_bt_2_jnt_capsule, center=<9.4921 -1.6737 6>, rotation=<rot 0 -10 90>}]
 dyn_L_coat_bt_3_jnt [{shapeType=Capsule, length=29, radius=9, shapeName=dyn_L_coat_bt_3_jnt_capsule, center=<9.4921 -1.6737 6>, rotation=<rot 0 -9.9999 90>}]
 dyn_L_coat_side_1_jnt [{shapeType=Capsule, length=29, radius=9, shapeName=dyn_R_coat_side_1_jnt_capsule, center=<10.0102 -0.7268 10>, rotation=<rot 9 -28 72>}]
 dyn_L_coat_side_2_jnt [{shapeType=Capsule, length=29, radius=9, shapeName=dyn_L_coat_side_2_jnt_capsule, center=<14.3063 -0.5094 6>, rotation=<rot 9 -28 72>}]
 dyn_L_coat_side_3_jnt [{shapeType=Capsule, length=29, radius=9, shapeName=dyn_L_coat_side_2_jnt_capsule, center=<16.658 1.2175 0>, rotation=<rot 9 -28 72>}]
 dyn_R_coat_bt_1_jnt [{shapeType=Capsule, length=29, radius=9, shapeName=dyn_R_coat_bt_1_jnt_capsule, center=<-9.1097 1.6063 -6>, rotation=<rot 0 -9.9999 90>}]
 dyn_R_coat_bt_2_jnt [{shapeType=Capsule, length=29, radius=9, shapeName=dyn_R_coat_bt_2_jnt_capsule, center=<-9.1097 1.6063 -6>, rotation=<rot 0 -9.9999 90>}]
 dyn_R_coat_bt_3_jnt [{shapeType=Capsule, length=29, radius=9, shapeName=dyn_R_coat_bt_3_jnt_capsule, center=<-9.1097 1.6063 -6>, rotation=<rot 0 -9.9999 90>}]
 dyn_R_coat_side_1_jnt [{
   shapeType=Capsule
   length=29
   radius=9
   shapeName=dyn_R_coat_side_1_jnt_capsule
   center=<-11.3493 1.1241 -10>
   rotation=<rot -9.8465 -28.481 72.5047>
  }]
 dyn_R_coat_side_2_jnt [{
   shapeType=Capsule
   length=29
   radius=9
   shapeName=dyn_R_coat_side_2_jnt_capsule
   center=<-15.3493 1.2326 -6>
   rotation=<rot -9.8465 -28.481 72.5047>
  }]
 dyn_R_coat_side_3_jnt [{
   shapeType=Capsule
   length=29
   radius=9
   shapeName=dyn_R_coat_side_2_jnt_capsule
   center=<-15.3493 -1.6908 0>
   rotation=<rot -9.8465 -28.481 72.5047>
  }]
 floor [
  {shapeType=Sphere, radius=200, shapeName=floor_sphere, center=<4.8592 -2.714 -190.1721>}
  {shapeType=Sphere, radius=22.9194, shapeName=floor_sphere, center=<0 0.5113 -6.6497>}
 ]
 foot_l [{shapeType=Capsule, length=15.6794, radius=7.0209, center=<-3.5001 8.3295 0.0009>, rotation=<rot 0 0 -90>}]
 foot_r [{shapeType=Capsule, length=16.7126, radius=6.7153, center=<6.4296 -5.0288 -0.0009>, rotation=<rot 0 0 90>}]
 master_C_coat_center_jnt [{shapeType=Capsule, length=0.505, radius=0.505, shapeName=master_C_coat_center_jnt_capsule, rotation=<rot 0 -90 0>}]
 master_L_coat_bt_jnt [{shapeType=Capsule, length=0.505, radius=0.505, shapeName=master_L_coat_bt_jnt_capsule, rotation=<rot 0 -90 0>}]
 master_L_coat_side_jnt [{shapeType=Capsule, length=0.505, radius=0.505, shapeName=master_L_coat_side_jnt_capsule, rotation=<rot 0 -90 0>}]
 master_R_coat_bt_jnt [{shapeType=Capsule, length=0.505, radius=0.505, shapeName=master_R_coat_bt_jnt_capsule, rotation=<rot 0 -90 0>}]
 master_R_coat_side_jnt [{shapeType=Capsule, length=0.505, radius=0.505, shapeName=master_R_coat_side_jnt_capsule, rotation=<rot 0 -90 0>}]
 pelvis [
  {shapeType=Sphere, radius=16.351, shapeName=pelvis_sphere, center=<-8.5464 0.6339 0>}
  {shapeType=Capsule, length=26.8862, radius=11.2057, center=<0.96 -1.1447 0.2284>, rotation=<rot 90 180 171.5626>}
 ]
 thigh_l [
  {shapeType=Capsule, length=32.1998, radius=8.2212, center=<-23.7318 -0.2 -0.255>, rotation=<rot 86.4154 0 177.6156>}
  {
   shapeType=Capsule
   length=24.9005
   radius=10.3613
   shapeName=thigh_l_capsule
   center=<-11.9227 0.0575 0.8477>
   rotation=<rot -87.1722 5.7146 -5.9387>
  }
 ]
 thigh_r [
  {shapeType=Capsule, length=31.9703, radius=8.4507, center=<23.7241 0.2 0.1312>, rotation=<rot 86.4154 0 177.6156>}
  {
   shapeType=Capsule
   length=29.6577
   radius=8.9282
   shapeName=thigh_r_capsule
   center=<11.9227 -0.0575 -0.8477>
   rotation=<rot 87.1762 174.5579 174.8909>
  }
 ]
}
constraints [
 {bone1Name=dyn_C_coat_center_1_jnt, bone2Name=master_C_coat_center_jnt, swing1LimitDegrees=11, swing2LimitDegrees=68, twistLimitDegrees=0}
 {bone1Name=dyn_C_coat_center_2_jnt, bone2Name=dyn_C_coat_center_1_jnt, swing1LimitDegrees=11, swing2LimitDegrees=68, twistLimitDegrees=0}
 {bone1Name=dyn_C_coat_center_3_jnt, bone2Name=dyn_C_coat_center_2_jnt, swing1LimitDegrees=11, swing2LimitDegrees=68, twistLimitDegrees=0}
 {bone1Name=dyn_L_coat_bt_1_jnt, bone2Name=master_L_coat_bt_jnt, swing1LimitDegrees=11, swing2LimitDegrees=68, twistLimitDegrees=0}
 {bone1Name=dyn_L_coat_bt_2_jnt, bone2Name=dyn_L_coat_bt_1_jnt, swing1LimitDegrees=11, swing2LimitDegrees=68, twistLimitDegrees=0}
 {bone1Name=dyn_L_coat_bt_3_jnt, bone2Name=dyn_L_coat_bt_2_jnt, swing1LimitDegrees=11, swing2LimitDegrees=68, twistLimitDegrees=0}
 {bone1Name=dyn_L_coat_side_1_jnt, bone2Name=master_L_coat_side_jnt, swing1LimitDegrees=11, swing2LimitDegrees=68, twistLimitDegrees=0}
 {bone1Name=dyn_L_coat_side_2_jnt, bone2Name=dyn_L_coat_side_1_jnt, swing1LimitDegrees=11, swing2LimitDegrees=68, twistLimitDegrees=0}
 {bone1Name=dyn_L_coat_side_3_jnt, bone2Name=dyn_L_coat_side_2_jnt, swing1LimitDegrees=11, swing2LimitDegrees=68, twistLimitDegrees=0}
 {bone1Name=dyn_R_coat_bt_1_jnt, bone2Name=master_R_coat_bt_jnt, swing1LimitDegrees=11, swing2LimitDegrees=68, twistLimitDegrees=0}
 {bone1Name=dyn_R_coat_bt_2_jnt, bone2Name=dyn_R_coat_bt_1_jnt, swing1LimitDegrees=11, swing2LimitDegrees=68, twistLimitDegrees=0}
 {bone1Name=dyn_R_coat_bt_3_jnt, bone2Name=dyn_R_coat_bt_2_jnt, swing1LimitDegrees=11, swing2LimitDegrees=68, twistLimitDegrees=0}
 {bone1Name=dyn_R_coat_side_1_jnt, bone2Name=master_R_coat_side_jnt, swing1LimitDegrees=11, swing2LimitDegrees=68, twistLimitDegrees=0}
 {bone1Name=dyn_R_coat_side_2_jnt, bone2Name=dyn_R_coat_side_1_jnt, swing1LimitDegrees=11, swing2LimitDegrees=68, twistLimitDegrees=0}
 {bone1Name=dyn_R_coat_side_3_jnt, bone2Name=dyn_R_coat_side_2_jnt, swing1LimitDegrees=11, swing2LimitDegrees=68, twistLimitDegrees=0}
 {
  bone1Name=dyn_C_coat_center_3_jnt
  bone2Name=dyn_R_coat_bt_3_jnt
  swing1LimitDegrees=45
  swing1Motion=Free
  swing2LimitDegrees=45
  swing2Motion=Free
  twistLimitDegrees=45
  twistMotion=Free
 }
 {
  bone1Name=dyn_R_coat_side_3_jnt
  bone2Name=dyn_R_coat_bt_3_jnt
  swing1LimitDegrees=45
  swing1Motion=Free
  swing2LimitDegrees=45
  swing2Motion=Free
  twistLimitDegrees=45
  twistMotion=Free
 }
 {
  bone1Name=dyn_L_coat_side_3_jnt
  bone2Name=dyn_L_coat_bt_3_jnt
  swing1LimitDegrees=45
  swing1Motion=Free
  swing2LimitDegrees=45
  swing2Motion=Free
  twistLimitDegrees=45
  twistMotion=Free
 }
 {
  bone1Name=dyn_R_coat_bt_2_jnt
  bone2Name=dyn_C_coat_center_2_jnt
  swing1LimitDegrees=45
  swing1Motion=Free
  swing2LimitDegrees=45
  swing2Motion=Free
  twistLimitDegrees=45
  twistMotion=Free
 }
]
