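# Post-process style cosmetic AnimBlueprint for the Ripped Harvester outfit, layering procedural motion onto the copied parent pose.
# Deform and outfit Control Rigs take movement states, with separate parachute inputs and a deploy curve for the matching glider.
# RigidBody physics with gravity override applies only with the Mat2 variant tag.
# A frontend curve blends slot animation onto the right pauldron bones.
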AnimBlueprint /BRCosmetics/Characters/Player/Male/Medium/Bodies/M_MED_RippedHarvester/Meshes/M_MED_RippedHarvester_AnimBP
target_skeleton=&/M_MED_RippedHarvester_Skeleton
graphs [
 AnimGraph {nodes [
   n1 = Layered blend per bone(
    BasePose=Use cached pose 'FinalDynPose'()
    BlendPoses_0=Slot 'DefaultSlot'()
    BlendWeights_0=Get PlayingFrontendIdle() -> (PlayingFrontendIdle:double)
   ) {layerSetup [{branchFilters [{boneName=R_shoulder_lower_pauldron_jnt}, {boneName=R_shoulder_upper_pauldron_jnt}]}]}
   Output Pose(Result=@n1) {layerGroup=None}
   n2 = Control Rig(Source=Copy Pose From Mesh() {bUseAttachedParent=true}) {
    alphaBoolBlend {blendOption=Linear}
    controlRigAssetReference {blueprintRigClass=&/M_MED_DeformRig_RippedHarvester_ControlRig.M_MED_DeformRig_RippedHarvester_ControlRig_C}
   }
   n3 = Control Rig!(
    Source=@n2
    Alpha
    Is Crouching CTRLRIG:double=Get Is Crouching() -> (Is Crouching:double)
    Is Skydiving CTRLRIG:double=Get Is Skydiving() -> (Is Skydiving:double)
    Is Ripped Harvester Glider Into Playing CTRLRIG:double=Get Is Ripped Harvester Glider Into Playing() -> (Is Ripped Harvester Glider Into Playing:double)
    Is Knee Sliding CTRLRIG:double=Get Is Knee Sliding() -> (Is Knee Sliding:double)
    Is Skydive Floating CTRLRIG:double=Get Is Skydive Floating() -> (Is Skydive Floating:double)
    Is Baseball Sliding CTRLRIG:double=Get Is Baseball Sliding() -> (Is Baseball Sliding:double)
    Is Parachute Open CTRLRIG:double=Get Is Parachute Open() -> (Is Parachute Open:double)
    Is Ripped Harvester Parachute Open CTRLRIG:double=Get Is Ripped Harvester Parachute Open() -> (Is Ripped Harvester Parachute Open:double)
   ) -> (Pose) {
    alphaBoolBlend {blendOption=Linear}
    controlRigAssetReference {blueprintRigClass=&/M_MED_RippedHarvester_CtrlRig.M_MED_RippedHarvester_CtrlRig_C}
   }
   Save cached pose 'Ripped Harvester'(Pose=@n3) {cacheName=Ripped Harvester}
   n4 = RigidBody(ComponentPose=Local To Component(LocalPose=Use cached pose 'Ripped Harvester'()), ExternalForce=Get GravityOverride()) {
    bClampLinearTranslationLimitToRefPose=true
    bOverrideWorldGravity=true
    overridePhysicsAsset=&/M_MED_RippedHarvester_Physics
    alphaBoolBlend {blendOption=Linear}
    componentAppliedLinearAccClamp=<3000 3000 3000>
    componentLinearAccScale=<0.5 0.2 0.5>
    componentLinearVelScale=<0.5 0.2 0.2>
    overrideWorldGravity {z=-980}
    simSpaceSettings {
     maxAngularAcceleration=500
     maxAngularVelocity=3
     maxLinearAcceleration=700
     velocityScaleZ=0
     worldAlpha=1
     externalAngularVelocity {x=0.8}
     externalLinearVelocity {x=0.3}
    }
   }
   Save cached pose 'FinalDynPose'(Pose=Blend Poses by bool(
     BlendPose_0=Component To Local(ComponentPose=@n4)
     BlendPose_1=Use cached pose 'Ripped Harvester'()
     bActiveValue=Get skirt_on/off() -> (skirt_on/off:bool)
    ) {blendType=Linear}) {cacheName=FinalDynPose}
  ]}
 EventGraph {nodes [
   Event BlueprintUpdateAnimation()
   > n1 = Sequence!(execute) -> (then_0, then_1, then_2:Exec)
   > n2 = Cast To FortPlayerPawn(Object=GetOwningActor())
   > Is Valid(InputObject=Get Ripped Harvester Glider Skelmesh() -> (Ripped Harvester Glider Skelmesh:Skeletal Mesh Object Reference))
   > n3 = Set Current Glider(Current Glider:Athena Glider Item Definition Object Reference=GetItemDefinitionTypeFromLoadoutWithActor!(
     Actor:Actor Object Reference=@n2
     LoadoutSlot:Gameplay Tag Structure=(TagName="LoadoutSlot.Glider")
     CosmeticType:Athena Cosmetic Item Definition Class Reference=/Script/FortniteGame.AthenaGliderItemDefinition
    ) -> (OutCosmeticItemDef:Athena Glider Item Definition Object Reference)) -> (Output_Get:Athena Glider Item Definition Object Reference)
   > n4 = Set Is Ripped Harvester Glider(Is Ripped Harvester Glider:bool="Equal (Object)"(A=@n3, B=Get Glider Item Def() -> (Glider Item Def:Athena Glider Item Definition Object Reference)) {is=title K2Node_CallFunction}) -> (Output_Get:bool)
   > Branch(Condition=@n4)
   n5 = GetAnimInstance(self=GetOwnerSkeletalMeshComponent())
   > Is Valid(InputObject=@n5)
   > Set Is Ripped Harvester Glider Into Playing(Is Ripped Harvester Glider Into Playing:double=GetCurveValue(self=@n5, CurveName=ParaGlideDeploy)) -> (Output_Get:double)
   n6 = Get Is Ripped Harvester Glider() -> (Is Ripped Harvester Glider:bool)
   n7 = Get bIsParachuteOpened()
   Set Is Ripped Harvester Parachute Open(execute=@n1.then_1, Is Ripped Harvester Parachute Open:double="To Float (Boolean)"(InBool=AND Boolean(A=@n7, B=@n6))) -> (Output_Get:double)
   n8 = NOT Boolean(A=Get Is Ripped Harvester Glider())
   n9 = Get bIsParachuteOpened()
   > Set Is Parachute Open(Is Parachute Open:double="To Float (Boolean)"(InBool=AND Boolean(A=@n9, B=@n8))) -> (Output_Get:double)
   > Set Is Skydiving(Is Skydiving:double="To Float (Boolean)"(InBool=Get bIsSkyDiving())) -> (Output_Get:double)
   > Set Is Crouching(Is Crouching:double="To Float (Boolean)"(InBool=Get bIsCrouching())) -> (Output_Get:double)
   > Set Is Knee Sliding(Is Knee Sliding:double="To Float (Boolean)"(InBool=Get bIsKneeSliding())) -> (Output_Get:double)
   > Set Is Baseball Sliding(Is Baseball Sliding:double="To Float (Boolean)"(InBool=Get bIsBaseballSliding())) -> (Output_Get:double)
   > Set Is Skydive Floating(Is Skydive Floating:double="To Float (Boolean)"(InBool=Get bIsSkydiveFloating())) -> (Output_Get:double)
   > n10 = Gravity Override State Function DEPRECATED(CharacterPartAnimInstance=Self-Reference(), GameTimeInSeconds=GetGameTimeInSeconds(), GravityOverrideParams=Get Gravity Override Params())
   > Set GravityOverride(GravityOverride=@n10)
   Branch(execute=@n1.then_2, Condition=Get bIsInFrontEnd())
   > Set PlayingFrontendIdle(PlayingFrontendIdle:double=GetCurveValue(self=GetOwnerFrontEndAnimInstance(), CurveName=Frontend_Idle_Bypass)) -> (Output_Get:double)
   Event OnCharacterCustomizationLoaded()
   > Set skirt_on/off(skirt_on/off:bool=CharacterItemDefinitionHasAnyMetaTag(Tags=Make Array("[0]"=(TagName="Cosmetics.Variant.Property.Mat2")))) -> (Output_Get:bool)
  ]}
]
variables [
 Anim_Inst {type=AnimInstance, default=null}
 Current Glider {type=AthenaGliderItemDefinition, default=null}
 Glider Item Def {type=AthenaGliderItemDefinition, default=/BRCosmetics/Athena/Items/Cosmetics/Gliders/Glider_RippedHarvester}
 Gravity Override Params {type=GravityOverrideParamsStruct}
 GravityOverride {type=Vector}
 Is Baseball Sliding {type=number}
 Is Crouching {type=number}
 Is Knee Sliding {type=number}
 Is Parachute Open {type=number}
 Is Ripped Harvester Glider {type=boolean}
 Is Ripped Harvester Glider Into Playing {type=number}
 Is Ripped Harvester Parachute Open {type=number}
 Is Skydive Floating {type=number}
 Is Skydiving {type=number}
 Is Wearing Ripped Harvester Glider {type=number}
 Pauldron Postition {type=number}
 PlayingFrontendIdle {type=number}
 Ripped Harvester Glider Skelmesh {type=SkeletalMesh, default=/BRCosmetics/Gadgets/Assets/VinderTech_GliderChute/Glider_RippedHarvester/Meshes/RippedHarvester_Glider}
 Ripped Harvester Glider Skelmesh Component {type=SkeletalMeshComponent}
 skirt_on/off {type=boolean}
]
