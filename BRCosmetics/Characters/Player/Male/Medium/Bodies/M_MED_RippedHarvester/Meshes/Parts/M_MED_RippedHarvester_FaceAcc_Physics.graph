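# Physics asset for the face accessory mesh of the male medium RippedHarvester cosmetic body, defining collision and joints for a ponytail.
# Capsule bodies cover the head, a tiny master ponytail root, a wider motion joint and three identical ponytail segments.
# Constraints chain them from head to tip: rotation is free at the head-to-root joint, tight at the motion joint, and looser along the segments with narrow twist limits.
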PhysicsAsset /BRCosmetics/Characters/Player/Male/Medium/Bodies/M_MED_RippedHarvester/Meshes/Parts/M_MED_RippedHarvester_FaceAcc_Physics
bodies {
 dyn_ponytail_1_jnt [{shapeType=Capsule, length=5.25, radius=4.75, shapeName=dyn_ponytail_1_jnt_capsule, center=<2.8983 0.0401 0>}]
 dyn_ponytail_2_jnt [{shapeType=Capsule, length=5.25, radius=4.75, shapeName=dyn_ponytail_2_jnt_capsule, center=<2.8983 0.0401 0>}]
 dyn_ponytail_3_jnt [{shapeType=Capsule, length=5.25, radius=4.75, shapeName=dyn_ponytail_3_jnt_capsule, center=<2.8983 0.0401 0>}]
 dyn_ponytail_motion_jnt [{
   shapeType=Capsule
   length=7.25
   radius=8.5
   shapeName=dyn_ponytail_motion_jnt_capsule
   center=<0.6592 -2.4654 0>
   rotation=<rot -90 0 0.0002>
  }]
 head [{
   shapeType=Capsule
   length=10.144
   radius=5.8607
   shapeName=head_capsule
   center=<5.5278 -0.5683 0.1114>
   rotation=<rot 5.1276 -89.1538 -87.0303>
  }]
 master_ponytail_jnt [{shapeType=Capsule, length=0.505, radius=0.505, shapeName=master_ponytail_jnt_capsule, rotation=<rot 0 -90 0>}]
}
constraints [
 {bone1Name=dyn_ponytail_3_jnt, bone2Name=dyn_ponytail_2_jnt, swing1LimitDegrees=25, swing2LimitDegrees=10, twistLimitDegrees=2}
 {bone1Name=dyn_ponytail_1_jnt, bone2Name=dyn_ponytail_motion_jnt, swing1LimitDegrees=25, swing2LimitDegrees=10, twistLimitDegrees=2}
 {bone1Name=dyn_ponytail_2_jnt, bone2Name=dyn_ponytail_1_jnt, swing1LimitDegrees=25, swing2LimitDegrees=10, twistLimitDegrees=2}
 {bone1Name=dyn_ponytail_motion_jnt, bone2Name=master_ponytail_jnt, swing1LimitDegrees=5, swing2LimitDegrees=5, twistLimitDegrees=0}
 {
  bone1Name=master_ponytail_jnt
  bone2Name=head
  swing1LimitDegrees=45
  swing1Motion=Free
  swing2LimitDegrees=45
  swing2Motion=Free
  twistLimitDegrees=45
  twistMotion=Free
 }
]
